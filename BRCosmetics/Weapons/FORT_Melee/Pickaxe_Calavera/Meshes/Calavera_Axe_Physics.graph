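# Physics asset for the Calavera two-handed pickaxe, giving its handle collision shapes and its dangling tassels a jointed physics setup.
# Capsule bodies cover the handle and tassel bones, with two four-bone chains (02-05 and 06-09) branching from dyn_tassel_MSTR.
# Constraints allow wide swing but very little twist along each chain.
# No constraint links the tassel master body to the handle.
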PhysicsAsset /BRCosmetics/Weapons/FORT_Melee/Pickaxe_Calavera/Meshes/Calavera_Axe_Physics
bodies {
 Melee_TwoHanded_Handle [
  {shapeType=Capsule, length=64.7011, radius=2.0321, center=<0.6111 -1.455 2.2756>, rotation=<rot -1.8901 -90.0388 -0.3293>}
  {shapeType=Capsule, length=1.3008, radius=4.0426, shapeName=Melee_TwoHanded_Handle_capsule, center=<0.8568 -0.8497 -31.2886>}
 ]
 dyn_tassel_02 [{shapeType=Capsule, length=1.0786, radius=1.1677, shapeName=dyn_tassel_02_capsule, center=<2.36 0 0>, rotation=<rot 0 90 90>}]
 dyn_tassel_03 [{shapeType=Capsule, length=1.4062, radius=1.4129, shapeName=dyn_tassel_03_capsule, center=<3 0 0>, rotation=<rot 0 90 90>}]
 dyn_tassel_04 [{shapeType=Capsule, length=2.5078, radius=1.8762, shapeName=dyn_tassel_04_capsule, center=<3.3506 0 0>, rotation=<rot 0 90 90>}]
 dyn_tassel_05 [{shapeType=Capsule, length=0.9058, radius=2.8625, shapeName=dyn_tassel_05_capsule, center=<3.3506 0 0>, rotation=<rot 0 90 90>}]
 dyn_tassel_06 [{shapeType=Capsule, length=0.7494, radius=0.873, shapeName=dyn_tassel_06_capsule, center=<1.8441 0 0>, rotation=<rot 0 90 90>}]
 dyn_tassel_07 [{shapeType=Capsule, length=0.4241, radius=1.2017, shapeName=dyn_tassel_07_capsule, center=<1 0 0>, rotation=<rot 0 90 90>}]
 dyn_tassel_08 [{shapeType=Capsule, length=0.9872, radius=1.4209, shapeName=dyn_tassel_08_capsule, center=<1 0 0>, rotation=<rot 0 90 90>}]
 dyn_tassel_09 [{shapeType=Capsule, length=0.806, radius=2.13, shapeName=dyn_tassel_09_capsule, center=<1.1404 0 0>, rotation=<rot 0 90 90>}]
 dyn_tassel_MSTR [{shapeType=Capsule, length=3.5026, radius=1.6063, shapeName=dyn_tassel_MSTR_capsule, rotation=<rot 0 90 90>}]
}
constraints [
 {bone1Name=dyn_tassel_06, bone2Name=dyn_tassel_MSTR, swing1LimitDegrees=45, swing2LimitDegrees=90, twistLimitDegrees=5}
 {bone1Name=dyn_tassel_07, bone2Name=dyn_tassel_06, swing1LimitDegrees=45, swing2LimitDegrees=65, twistLimitDegrees=5}
 {bone1Name=dyn_tassel_08, bone2Name=dyn_tassel_07, swing1LimitDegrees=65, swing2LimitDegrees=65, twistLimitDegrees=5}
 {bone1Name=dyn_tassel_09, bone2Name=dyn_tassel_08, swing1LimitDegrees=65, swing2LimitDegrees=65, twistLimitDegrees=5}
 {bone1Name=dyn_tassel_02, bone2Name=dyn_tassel_MSTR, swing1LimitDegrees=35, swing2LimitDegrees=90, twistLimitDegrees=5}
 {bone1Name=dyn_tassel_03, bone2Name=dyn_tassel_02, swing1LimitDegrees=45, swing2LimitDegrees=65, twistLimitDegrees=5}
 {bone1Name=dyn_tassel_04, bone2Name=dyn_tassel_03, swing1LimitDegrees=45, swing2LimitDegrees=45, twistLimitDegrees=5}
 {bone1Name=dyn_tassel_05, bone2Name=dyn_tassel_04, swing1LimitDegrees=45, swing2LimitDegrees=45, twistLimitDegrees=5}
]
