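# Animation Blueprint for the Calavera pickaxe that adds secondary physics motion to its mesh.
# The input pose runs through a RigidBody simulation in the Melee_TwoHanded_Handle bone's space, using the axe's physics asset.
# Additive yaw on dyn_handle_r and dyn_handle_l blends in while Was Jumping is true, then sprung, angle-limited AnimDynamics simulates both bones.
# The event graph caches the owning PlayerPawn_Athena and copies its bWasJumping flag every update.
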AnimBlueprint /BRCosmetics/Weapons/FORT_Melee/Pickaxe_Calavera/Meshes/Calavera_Axe_AnimBP
target_skeleton=&/Calavera_Axe_Skeleton
graphs [
 AnimGraph {nodes [
   AnimGraphNode_SubInput_0 = Input Pose()
   n1 = RigidBody(ComponentPose=Local To Component(LocalPose=@AnimGraphNode_SubInput_0)) {
    bClampLinearTranslationLimitToRefPose=true
    bOverrideWorldGravity=true
    overridePhysicsAsset=&/Calavera_Axe_Physics
    simulationSpace=BaseBoneSpace
    alphaBoolBlend {blendOption=Linear}
    baseBoneRef {boneName=Melee_TwoHanded_Handle}
    componentAppliedLinearAccClamp=<2000 2000 2000>
    componentLinearAccScale=<0.4 0.3 0.3>
    componentLinearVelScale=<0.4 0.3 0.3>
    overrideWorldGravity {z=-980}
   }
   n2 = Get Was Jumping() -> (Was Jumping:bool)
   n3 = "Transform (Modify) Bone - Bone: dyn_handle_r"!(ComponentPose=@n1, bAlphaBoolEnabled:bool=@n2, Translation="0,0,0", Rotation="0,12,0", Scale="1,1,1") -> (Pose) {
    alphaInputType=Bool
    rotationMode=BMM_Additive
    rotationSpace=BCS_BoneSpace
    alphaBoolBlend {blendInTime=0.1, blendOption=ExpInOut, blendOutTime=0.5}
    boneToModify {boneName=dyn_handle_r}
    rotation {yaw=12}
   }
   n4 = "Transform (Modify) Bone - Bone: dyn_handle_l"!(ComponentPose=@n3, bAlphaBoolEnabled:bool=@n2, Translation="0,0,0", Rotation="0,12,0", Scale="1,1,1") -> (Pose) {
    alphaInputType=Bool
    rotationMode=BMM_Additive
    rotationSpace=BCS_BoneSpace
    alphaBoolBlend {blendInTime=0.1, blendOption=ExpInOut, blendOutTime=0.5}
    boneToModify {boneName=dyn_handle_l}
    rotation {yaw=12}
   }
   n5 = Anim Dynamics - Bone: dyn_handle_l!(ComponentPose=@n4, Alpha=1, GravityOverride:Vector="0,0,0", ExternalForce="0,0,0") -> (Pose) {
    angularBiasOverride=0.99
    angularDampingOverride=0.93
    angularSpringConstant=1100
    bAngularSpring=true
    bOverrideAngularBias=true
    bOverrideAngularDamping=true
    bUseGravityOverride=true
    simulationSpace=RootRelative
    alphaBoolBlend {blendOption=Linear}
    alphaScaleBiasClamp {bInterpResult=true}
    boundBone {boneName=dyn_handle_l}
    componentAppliedLinearAccClamp=<1500 1500 1500>
    componentLinearAccScale=<0.3 0.3 0.3>
    componentLinearVelScale=<0.3 0.3 0.3>
    gravityOverride {z=-980}
    physicsBodyDefinitions [{
      boundBone {boneName=dyn_handle_l}
      boxExtents=<20 10 10>
      constraintSetup {angularLimitsMax=<0 0 30>, angularLimitsMin=<0 0 -10>, angularTarget=<1 0 0>}
      localJointOffset=<0.4 0 0>
     }]
   }
   n6 = Anim Dynamics - Bone: dyn_handle_r!(ComponentPose=@n5, Alpha=1, GravityOverride:Vector="0,0,0", ExternalForce="0,0,0") -> (Pose) {
    angularBiasOverride=0.99
    angularDampingOverride=0.93
    angularSpringConstant=1100
    bAngularSpring=true
    bOverrideAngularBias=true
    bOverrideAngularDamping=true
    bUseGravityOverride=true
    simulationSpace=RootRelative
    alphaBoolBlend {blendOption=Linear}
    alphaScaleBiasClamp {bInterpResult=true}
    boundBone {boneName=dyn_handle_r}
    componentAppliedLinearAccClamp=<1500 1500 1500>
    componentLinearAccScale=<0.3 0.3 0.3>
    componentLinearVelScale=<0.3 0.3 0.3>
    gravityOverride {z=-980}
    physicsBodyDefinitions [{
      boundBone {boneName=dyn_handle_r}
      boxExtents=<20 10 10>
      constraintSetup {angularLimitsMax=<0 0 30>, angularLimitsMin=<0 0 -10>, angularTarget=<1 0 0>}
      localJointOffset=<0.4 0 0>
     }]
   }
   Output Pose(Result=Component To Local(ComponentPose=@n6)) {layerGroup=None}
  ]}
 EventGraph {nodes [
   Event BlueprintUpdateAnimation()
   > Set Was Jumping(Was Jumping:bool=Get bWasJumping!(self:Character Object Reference=Get As Player Pawn Athena() -> (As Player Pawn Athena:Player Pawn Athena Object Reference)) -> (bWasJumping:bool)) -> (Output_Get:bool)
   Event BlueprintInitializeAnimation()
   > n1 = Cast To PlayerPawn_Athena(Object=TryGetPawnOwner())
   > Set As Player Pawn Athena(As Player Pawn Athena:Player Pawn Athena Object Reference=@n1) -> (Output_Get:Player Pawn Athena Object Reference)
  ]}
]
variables [
 As Player Pawn Athena {type=PlayerPawn_Athena_C}
 Was Jumping {type=boolean}
]
